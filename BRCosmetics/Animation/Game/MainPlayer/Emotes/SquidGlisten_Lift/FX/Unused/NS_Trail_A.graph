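# Niagara ribbon trail for the SquidGlisten_Lift emote, stored in an Unused folder, that follows the attach parent's weapon_l bone.
# Sprite emitter Position001 tracks the bone as a leader particle, feeding three layered ribbons (Red_Base, Red_Outline, Red_Streak) through attribute readers.
# User animation-curve floats drive spawn rate and initial opacity, while ribbons taper and shift from red to purple as they fade.
# Yellow and left/right offset or socket-based variants are disabled.
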NiagaraSystem /BRCosmetics/Animation/Game/MainPlayer/Emotes/SquidGlisten_Lift/FX/Unused/NS_Trail_A
dependencies {
 usedDataInterfaces [
  &NiagaraDataInterfaceParticleRead
  &NiagaraDataInterfaceColorCurve
  &NiagaraDataInterfaceCurve
  &NiagaraDataInterfaceSkeletalMesh
  &NiagaraDataInterfaceArrayFloat
 ]
 usedDynamicInputs [&FNTimeOfDayColor, &SelectFloatFromArray]
 usedModules [
  &SystemState
  &self:SystemSpawnScript.SystemScriptSource.SystemScriptGraph.NiagaraNodeAssignment_0.SetVariables_#1
 ]
}
emitters {
 L_Position {
  bEnabled=false
  emitterUpdateScript [
   &EmitterState(Life Cycle Mode=::NewEnumerator1, Loop Behavior=::NewEnumerator1, Loop Duration=3.5, UseLoopDelay=false)
   &SpawnRate(SpawnRate=15, Use Spawn Probability=false) {enabled=false}
   &SpawnBurst_Instantaneous(Spawn Time=1, Use Spawn Probability=false, UseLoopCountLimit=false)
  ]
  particleSpawnScript [&InitializeParticle(
    Color Mode=::NewEnumerator0
    Lifetime=2.5
    Sprite Size Mode=::NewEnumerator3
    Uniform Sprite Size=5
    UsePositionOffset=false
    Write Material Random=true
   )]
  particleUpdateScript [
   &ParticleState()
   &SkeletalMeshLocation(
    Bone Sampling Mode=::NewEnumerator4
    MaskSpawnGroup=false
    Sampled Orientation Coordinate Space=::Local
    Sampled Position Coordinate Space=::Local
    Sampled Position Offset=<15 0 0>
    Sampled Position Offset Coordinate Space:ENiagaraSkelMeshTransforms=::NewEnumerator2
    Sampled Velocity Coordinate Space:ENiagaraCoordinateSpace=::Local
    Skeletal Mesh=DataInterface() {propertyValues {
      DefaultMesh=null
      ExcludeBoneName=root
      OptionalComponentClass=null
      PreviewMesh=null
      SkinningMode=SkinOnTheFly
      SoftSourceActor=null
      SourceMode=AttachParent
      UvSetIndex=0
      WholeMeshLOD=-1
      bExcludeBone=false
      bReadDeformedGeometry=true
      bRequireCurrentFrameData=true
      ComponentTags []
      FilteredBones [weapon_l]
      FilteredSockets []
      MeshUserParameter {parameter {name=None, typeDefHandle {registeredTypeIndex=1}, varData []}}
      SamplingRegions []
     }}
   ) {module=SkeletalMeshLocation001}
   /SetVariables(Particles.Position=&AddVectorToPosition(Position=$Particles.SkeletalMeshLocation001.SampledPosition, Vector=<10 0 0>)) {enabled=false, isSetParameters=true, module=SetVariables_#2}
  ]
  rendererClasses [&NiagaraSpriteRendererProperties]
  renderers [{rendererClass=&NiagaraSpriteRendererProperties}]
 }
 L_Position_Spawn {
  bEnabled=false
  emitterUpdateScript [
   &EmitterState(Life Cycle Mode=::NewEnumerator1, Loop Behavior=::NewEnumerator1, Loop Duration=1, UseLoopDelay=false)
   &SpawnRate(SpawnRate=15, Use Spawn Probability=false) {enabled=false}
   &SpawnBurst_Instantaneous(Use Spawn Probability=false, UseLoopCountLimit=false)
  ]
  particleSpawnScript [&InitializeParticle(Color Mode=::NewEnumerator0, Sprite Size Mode=::NewEnumerator3, Uniform Sprite Size=5, UsePositionOffset=false, Write Material Random=true)]
  particleUpdateScript [
   &ParticleState()
   &SkeletalMeshLocation(
    MaskSpawnGroup=false
    Mesh Sampling Type=::NewEnumerator5
    Sampled Orientation Coordinate Space=::Local
    Sampled Position Coordinate Space=::Local
    Sampled Position Offset Coordinate Space:ENiagaraSkelMeshTransforms=::NewEnumerator2
    Sampled Velocity Coordinate Space:ENiagaraCoordinateSpace=::Local
    Skeletal Mesh=DataInterface() {propertyValues {
      DefaultMesh=null
      ExcludeBoneName=root
      OptionalComponentClass=null
      PreviewMesh=null
      SkinningMode=SkinOnTheFly
      SoftSourceActor=null
      SourceMode=Default
      UvSetIndex=0
      WholeMeshLOD=-1
      bExcludeBone=false
      bReadDeformedGeometry=true
      bRequireCurrentFrameData=true
      ComponentTags []
      FilteredBones []
      FilteredSockets [l_weapon_lSocket]
      MeshUserParameter {parameter {name=None, typeDefHandle {registeredTypeIndex=1}, varData []}}
      SamplingRegions []
     }}
    Socket Sampling Mode=::NewEnumerator5
   ) {module=SkeletalMeshLocation001}
   /SetVariables(Particles.Position=&AddVectorToPosition(Position=$Particles.SkeletalMeshLocation001.SampledPosition, Vector=<10 0 0>)) {enabled=false, isSetParameters=true, module=SetVariables_#2}
  ]
  rendererClasses [&NiagaraSpriteRendererProperties]
  renderers [{rendererClass=&NiagaraSpriteRendererProperties}]
 }
 L_Small_Yellow_Base {
  bEnabled=false
  emitterUpdateScript [
   &EmitterState(Life Cycle Mode=::NewEnumerator1, Loop Duration=15, UseLoopDelay=false)
   &SpawnRate(SpawnRate=&Multiply_Float(A=50, B=$User.Anim_Curve_SpawnRate), Use Spawn Probability=false)
   &SpawnParticlesFromOtherEmitter(
    Attribute Reader=DataInterface() {propertyValues {EmitterBinding {bindingMode=Other, emitterName=L_Position}}}
    Spawn Rate=0
    Use Maximum Age=false
    Use Minimum Age=false
   )
  ]
  particleSpawnScript [
   &InitializeParticle(
    Color=&ColorFromCurve(
     ColorCurve=DataInterface() {propertyValues {
       AlphaCurve {keys [<at 0 0 RCIM_Cubic>,<at 1 1 RCIM_Cubic>]}
       BlueCurve {keys [<at 0 1 RCIM_Cubic>,<at 1 1 RCIM_Cubic>]}
       GreenCurve {keys [<at 0 1 RCIM_Cubic>,<at 1 1 RCIM_Cubic>]}
       RedCurve {keys [<at 0 1 RCIM_Cubic>,<at 1 1 RCIM_Cubic>]}
      }}
     CurveIndex=$User.Anim_Curve_InitialOpacity
    )
    Lifetime=0.25
    Ribbon Width=2
    Ribbon Width Mode=::NewEnumerator1
    UsePositionOffset=false
    Write Material Random=true
   )
   &SampleParticlesFromOtherEmitter()
  ]
  particleUpdateScript [
   &ParticleState()
   &SolveForcesAndVelocity(Clamp Velocity=false, Limit Acceleration=false)
   &ScaleRibbonWidth(Ribbon Width Scale=&FloatFromCurve(FloatCurve=DataInterface() {propertyValues {Curve {keys [<at 0 1>,<at 1 0>]}}}))
   &ScaleColor(
    Linear Color Curve=DataInterface() {propertyValues {
      AlphaCurve {keys [<at 0 0>,<at 0.108 1>,<at 1 0>]}
      BlueCurve {keys [<at 0.095 0>,<at 0.4749 0>,<at 1 0.6458>]}
      GreenCurve {keys [<at 0.095 0.7167>,<at 0.4749 0.1537>,<at 1 0>]}
      RedCurve {keys [<at 0.095 1>,<at 0.4749 0.974>,<at 1 0.296>]}
     }}
    Scale Mode=::NewEnumerator2
   ) {module=ScaleColor001}
  ]
  rendererClasses [&NiagaraRibbonRendererProperties]
  renderers [{rendererClass=&NiagaraRibbonRendererProperties}]
 }
 L_Small_Yellow_Base001 {
  bEnabled=false
  emitterUpdateScript [
   &EmitterState(Life Cycle Mode=::NewEnumerator1, Loop Duration=15, UseLoopDelay=false)
   &SpawnRate(SpawnRate=&Multiply_Float(A=50, B=$User.Anim_Curve_SpawnRate), Use Spawn Probability=false)
   &SpawnParticlesFromOtherEmitter(
    Attribute Reader=DataInterface() {propertyValues {EmitterBinding {bindingMode=Other, emitterName=L_Position_Spawn}}}
    Spawn Rate=0
    Use Maximum Age=false
    Use Minimum Age=false
   )
  ]
  particleSpawnScript [
   &InitializeParticle(
    Color=&ColorFromCurve(
     ColorCurve=DataInterface() {propertyValues {
       AlphaCurve {keys [<at 0 0 RCIM_Cubic>,<at 1 1 RCIM_Cubic>]}
       BlueCurve {keys [<at 0 1 RCIM_Cubic>,<at 1 1 RCIM_Cubic>]}
       GreenCurve {keys [<at 0 1 RCIM_Cubic>,<at 1 1 RCIM_Cubic>]}
       RedCurve {keys [<at 0 1 RCIM_Cubic>,<at 1 1 RCIM_Cubic>]}
      }}
     CurveIndex=$User.Anim_Curve_InitialOpacity
    )
    Lifetime=0.25
    Ribbon Width=2
    Ribbon Width Mode=::NewEnumerator1
    UsePositionOffset=false
    Write Material Random=true
   )
   &SampleParticlesFromOtherEmitter()
  ]
  particleUpdateScript [
   &ParticleState()
   &SolveForcesAndVelocity(Clamp Velocity=false, Limit Acceleration=false)
   &ScaleRibbonWidth(Ribbon Width Scale=&FloatFromCurve(FloatCurve=DataInterface() {propertyValues {Curve {keys [<at 0 1>,<at 1 0>]}}}))
   &ScaleColor(
    Linear Color Curve=DataInterface() {propertyValues {
      AlphaCurve {keys [<at 0 0>,<at 0.108 1>,<at 1 0>]}
      BlueCurve {keys [<at 0.095 0>,<at 0.4749 0>,<at 1 0.6458>]}
      GreenCurve {keys [<at 0.095 0.7167>,<at 0.4749 0.1537>,<at 1 0>]}
      RedCurve {keys [<at 0.095 1>,<at 0.4749 0.974>,<at 1 0.296>]}
     }}
    Scale Mode=::NewEnumerator2
   ) {module=ScaleColor001}
  ]
  rendererClasses [&NiagaraRibbonRendererProperties]
  renderers [{rendererClass=&NiagaraRibbonRendererProperties}]
 }
 Position {
  bEnabled=false
  emitterUpdateScript [
   &EmitterState(Life Cycle Mode=::NewEnumerator1, Loop Duration=15, UseLoopDelay=false)
   &SpawnBurst_Instantaneous(Use Spawn Probability=false, UseLoopCountLimit=false)
  ]
  particleSpawnScript [&InitializeParticle(Color Mode=::NewEnumerator0, Lifetime=15, UsePositionOffset=false, Write Material Random=true)]
  particleUpdateScript [
   &ParticleState()
   &SkeletalMeshLocation(
    Bone Sampling Mode=::NewEnumerator4
    MaskSpawnGroup=false
    Orientation Sampling=::NewEnumerator1
    Skeletal Mesh=DataInterface() {propertyValues {
      DefaultMesh=null
      ExcludeBoneName=root
      OptionalComponentClass=null
      PreviewMesh=null
      SkinningMode=SkinOnTheFly
      SoftSourceActor=null
      SourceMode=AttachParent
      UvSetIndex=0
      WholeMeshLOD=-1
      bExcludeBone=false
      bReadDeformedGeometry=true
      bRequireCurrentFrameData=true
      ComponentTags []
      FilteredBones [weapon_l]
      FilteredSockets []
      MeshUserParameter {parameter {name=None, typeDefHandle {registeredTypeIndex=1}, varData []}}
      SamplingRegions []
     }}
   ) {module=SkeletalMeshLocation001}
  ]
  rendererClasses [&NiagaraSpriteRendererProperties]
  renderers [{rendererClass=&NiagaraSpriteRendererProperties}]
 }
 Position001 {
  emitterUpdateScript [
   &EmitterState(Life Cycle Mode=::NewEnumerator1, Loop Duration=15, UseLoopDelay=false)
   &SpawnBurst_Instantaneous(Use Spawn Probability=false, UseLoopCountLimit=false)
  ]
  particleSpawnScript [&InitializeParticle(
    Color=<rgba 0.3542 0.6261 1 0.2>
    Lifetime=15
    Sprite Size Mode=::NewEnumerator3
    Uniform Sprite Size=30
    UsePositionOffset=false
    Write Material Random=true
   )]
  particleUpdateScript [
   &ParticleState()
   &SkeletalMeshLocation(
    Bone Sampling Mode=::NewEnumerator4
    MaskSpawnGroup=false
    Orientation Sampling=::NewEnumerator1
    Skeletal Mesh=DataInterface() {propertyValues {
      DefaultMesh=null
      ExcludeBoneName=root
      OptionalComponentClass=null
      PreviewMesh=null
      SkinningMode=SkinOnTheFly
      SoftSourceActor=null
      SourceMode=AttachParent
      UvSetIndex=0
      WholeMeshLOD=-1
      bExcludeBone=false
      bReadDeformedGeometry=true
      bRequireCurrentFrameData=true
      ComponentTags []
      FilteredBones [weapon_l]
      FilteredSockets []
      MeshUserParameter {parameter {name=None, typeDefHandle {registeredTypeIndex=1}, varData []}}
      SamplingRegions []
     }}
   ) {module=SkeletalMeshLocation001}
  ]
  rendererClasses [&NiagaraSpriteRendererProperties]
  renderers [{rendererClass=&NiagaraSpriteRendererProperties}]
 }
 R_Position {
  bEnabled=false
  emitterUpdateScript [
   &EmitterState(Life Cycle Mode=::NewEnumerator1, Loop Behavior=::NewEnumerator1, Loop Duration=3.5, UseLoopDelay=false)
   &SpawnRate(SpawnRate=15, Use Spawn Probability=false) {enabled=false}
   &SpawnBurst_Instantaneous(Spawn Time=1, Use Spawn Probability=false, UseLoopCountLimit=false)
  ]
  particleSpawnScript [&InitializeParticle(
    Color Mode=::NewEnumerator0
    Lifetime=2.5
    Sprite Size Mode=::NewEnumerator3
    Uniform Sprite Size=5
    UsePositionOffset=false
    Write Material Random=true
   )]
  particleUpdateScript [
   &ParticleState()
   &SkeletalMeshLocation(
    Bone Sampling Mode=::NewEnumerator4
    MaskSpawnGroup=false
    Sampled Orientation Coordinate Space=::Local
    Sampled Position Coordinate Space=::Local
    Sampled Position Offset=<-15 0 0>
    Sampled Position Offset Coordinate Space:ENiagaraSkelMeshTransforms=::NewEnumerator2
    Sampled Velocity Coordinate Space:ENiagaraCoordinateSpace=::Local
    Skeletal Mesh=DataInterface() {propertyValues {
      DefaultMesh=null
      ExcludeBoneName=root
      OptionalComponentClass=null
      PreviewMesh=null
      SkinningMode=SkinOnTheFly
      SoftSourceActor=null
      SourceMode=AttachParent
      UvSetIndex=0
      WholeMeshLOD=-1
      bExcludeBone=false
      bReadDeformedGeometry=true
      bRequireCurrentFrameData=true
      ComponentTags []
      FilteredBones [weapon_l]
      FilteredSockets []
      MeshUserParameter {parameter {name=None, typeDefHandle {registeredTypeIndex=1}, varData []}}
      SamplingRegions []
     }}
   ) {module=SkeletalMeshLocation001}
   /SetVariables(Particles.Position=&AddVectorToPosition(Position=$Particles.SkeletalMeshLocation001.SampledPosition, Vector=<-10 0 0>)) {enabled=false, isSetParameters=true, module=SetVariables_#3}
  ]
  rendererClasses [&NiagaraSpriteRendererProperties]
  renderers [{rendererClass=&NiagaraSpriteRendererProperties}]
 }
 R_Position_Spawn {
  bEnabled=false
  emitterUpdateScript [
   &EmitterState(Life Cycle Mode=::NewEnumerator1, Loop Behavior=::NewEnumerator1, Loop Duration=1, UseLoopDelay=false)
   &SpawnRate(SpawnRate=15, Use Spawn Probability=false) {enabled=false}
   &SpawnBurst_Instantaneous(Use Spawn Probability=false, UseLoopCountLimit=false)
  ]
  particleSpawnScript [&InitializeParticle(Color Mode=::NewEnumerator0, Sprite Size Mode=::NewEnumerator3, Uniform Sprite Size=5, UsePositionOffset=false, Write Material Random=true)]
  particleUpdateScript [
   &ParticleState()
   &SkeletalMeshLocation(
    MaskSpawnGroup=false
    Mesh Sampling Type=::NewEnumerator5
    Sampled Orientation Coordinate Space=::Local
    Sampled Position Coordinate Space=::Local
    Sampled Position Offset Coordinate Space:ENiagaraSkelMeshTransforms=::NewEnumerator2
    Sampled Velocity Coordinate Space:ENiagaraCoordinateSpace=::Local
    Skeletal Mesh=DataInterface() {propertyValues {
      DefaultMesh=null
      ExcludeBoneName=root
      OptionalComponentClass=null
      PreviewMesh=null
      SkinningMode=SkinOnTheFly
      SoftSourceActor=null
      SourceMode=Default
      UvSetIndex=0
      WholeMeshLOD=-1
      bExcludeBone=false
      bReadDeformedGeometry=true
      bRequireCurrentFrameData=true
      ComponentTags []
      FilteredBones []
      FilteredSockets [r_weapon_lSocket]
      MeshUserParameter {parameter {name=None, typeDefHandle {registeredTypeIndex=1}, varData []}}
      SamplingRegions []
     }}
    Socket Sampling Mode=::NewEnumerator5
   ) {module=SkeletalMeshLocation001}
   /SetVariables(Particles.Position=&AddVectorToPosition(Position=$Particles.SkeletalMeshLocation001.SampledPosition, Vector=<-10 0 0>)) {enabled=false, isSetParameters=true, module=SetVariables_#3}
  ]
  rendererClasses [&NiagaraSpriteRendererProperties]
  renderers [{rendererClass=&NiagaraSpriteRendererProperties}]
 }
 R_Small_Yellow_Base {
  bEnabled=false
  emitterUpdateScript [
   &EmitterState(Life Cycle Mode=::NewEnumerator1, Loop Duration=15, UseLoopDelay=false)
   &SpawnRate(SpawnRate=&Multiply_Float(A=50, B=$User.Anim_Curve_SpawnRate), Use Spawn Probability=false)
   &SpawnParticlesFromOtherEmitter(
    Attribute Reader=DataInterface() {propertyValues {EmitterBinding {bindingMode=Other, emitterName=R_Position}}}
    Spawn Rate=0
    Use Maximum Age=false
    Use Minimum Age=false
   )
  ]
  particleSpawnScript [
   &InitializeParticle(
    Color=&ColorFromCurve(
     ColorCurve=DataInterface() {propertyValues {
       AlphaCurve {keys [<at 0 0 RCIM_Cubic>,<at 1 1 RCIM_Cubic>]}
       BlueCurve {keys [<at 0 1 RCIM_Cubic>,<at 1 1 RCIM_Cubic>]}
       GreenCurve {keys [<at 0 1 RCIM_Cubic>,<at 1 1 RCIM_Cubic>]}
       RedCurve {keys [<at 0 1 RCIM_Cubic>,<at 1 1 RCIM_Cubic>]}
      }}
     CurveIndex=$User.Anim_Curve_InitialOpacity
    )
    Lifetime=0.25
    Ribbon Width=2
    Ribbon Width Mode=::NewEnumerator1
    UsePositionOffset=false
    Write Material Random=true
   )
   &SampleParticlesFromOtherEmitter()
  ]
  particleUpdateScript [
   &ParticleState()
   &SolveForcesAndVelocity(Clamp Velocity=false, Limit Acceleration=false)
   &ScaleRibbonWidth(Ribbon Width Scale=&FloatFromCurve(FloatCurve=DataInterface() {propertyValues {Curve {keys [<at 0 1>,<at 1 0>]}}}))
   &ScaleColor(
    Linear Color Curve=DataInterface() {propertyValues {
      AlphaCurve {keys [<at 0 0>,<at 0.108 1>,<at 1 0>]}
      BlueCurve {keys [<at 0.095 0>,<at 0.4749 0>,<at 1 0.6458>]}
      GreenCurve {keys [<at 0.095 0.7167>,<at 0.4749 0.1537>,<at 1 0>]}
      RedCurve {keys [<at 0.095 1>,<at 0.4749 0.974>,<at 1 0.296>]}
     }}
    Scale Mode=::NewEnumerator2
   ) {module=ScaleColor001}
  ]
  rendererClasses [&NiagaraRibbonRendererProperties]
  renderers [{rendererClass=&NiagaraRibbonRendererProperties}]
 }
 R_Small_Yellow_Base001 {
  bEnabled=false
  emitterUpdateScript [
   &EmitterState(Life Cycle Mode=::NewEnumerator1, Loop Duration=15, UseLoopDelay=false)
   &SpawnRate(SpawnRate=&Multiply_Float(A=50, B=$User.Anim_Curve_SpawnRate), Use Spawn Probability=false)
   &SpawnParticlesFromOtherEmitter(
    Attribute Reader=DataInterface() {propertyValues {EmitterBinding {bindingMode=Other, emitterName=R_Position_Spawn}}}
    Spawn Rate=0
    Use Maximum Age=false
    Use Minimum Age=false
   )
  ]
  particleSpawnScript [
   &InitializeParticle(
    Color=&ColorFromCurve(
     ColorCurve=DataInterface() {propertyValues {
       AlphaCurve {keys [<at 0 0 RCIM_Cubic>,<at 1 1 RCIM_Cubic>]}
       BlueCurve {keys [<at 0 1 RCIM_Cubic>,<at 1 1 RCIM_Cubic>]}
       GreenCurve {keys [<at 0 1 RCIM_Cubic>,<at 1 1 RCIM_Cubic>]}
       RedCurve {keys [<at 0 1 RCIM_Cubic>,<at 1 1 RCIM_Cubic>]}
      }}
     CurveIndex=$User.Anim_Curve_InitialOpacity
    )
    Lifetime=0.25
    Ribbon Width=2
    Ribbon Width Mode=::NewEnumerator1
    UsePositionOffset=false
    Write Material Random=true
   )
   &SampleParticlesFromOtherEmitter()
  ]
  particleUpdateScript [
   &ParticleState()
   &SolveForcesAndVelocity(Clamp Velocity=false, Limit Acceleration=false)
   &ScaleRibbonWidth(Ribbon Width Scale=&FloatFromCurve(FloatCurve=DataInterface() {propertyValues {Curve {keys [<at 0 1>,<at 1 0>]}}}))
   &ScaleColor(
    Linear Color Curve=DataInterface() {propertyValues {
      AlphaCurve {keys [<at 0 0>,<at 0.108 1>,<at 1 0>]}
      BlueCurve {keys [<at 0.095 0>,<at 0.4749 0>,<at 1 0.6458>]}
      GreenCurve {keys [<at 0.095 0.7167>,<at 0.4749 0.1537>,<at 1 0>]}
      RedCurve {keys [<at 0.095 1>,<at 0.4749 0.974>,<at 1 0.296>]}
     }}
    Scale Mode=::NewEnumerator2
   ) {module=ScaleColor001}
  ]
  rendererClasses [&NiagaraRibbonRendererProperties]
  renderers [{rendererClass=&NiagaraRibbonRendererProperties}]
 }
 Red_Base {
  emitterUpdateScript [
   &EmitterState(Life Cycle Mode=::NewEnumerator1, Loop Duration=15, UseLoopDelay=false)
   &SpawnRate(SpawnRate=&Multiply_Float(A=50, B=$User.Anim_Curve_SpawnRate), Use Spawn Probability=false)
   &SpawnParticlesFromOtherEmitter(
    Attribute Reader=DataInterface() {propertyValues {EmitterBinding {bindingMode=Other, emitterName=Position001}}}
    Spawn Rate=0
    Use Maximum Age=false
    Use Minimum Age=false
   )
  ]
  particleSpawnScript [
   &InitializeParticle(
    Color=&ColorFromCurve(
     ColorCurve=DataInterface() {propertyValues {
       AlphaCurve {keys [<at 0 0 RCIM_Cubic>,<at 1 1 RCIM_Cubic>]}
       BlueCurve {keys [<at 0 1 RCIM_Cubic>,<at 1 1 RCIM_Cubic>]}
       GreenCurve {keys [<at 0 1 RCIM_Cubic>,<at 1 1 RCIM_Cubic>]}
       RedCurve {keys [<at 0 1 RCIM_Cubic>,<at 1 1 RCIM_Cubic>]}
      }}
     CurveIndex=$User.Anim_Curve_InitialOpacity
    )
    Lifetime=0.45
    Ribbon Width=22
    Ribbon Width Mode=::NewEnumerator1
    UsePositionOffset=false
    Write Material Random=true
   )
   &SampleParticlesFromOtherEmitter()
  ]
  particleUpdateScript [
   &ParticleState()
   &SolveForcesAndVelocity(Clamp Velocity=false, Limit Acceleration=false)
   &ScaleRibbonWidth(Ribbon Width Scale=&FloatFromCurve(FloatCurve=DataInterface() {propertyValues {Curve {keys [<at 0 1>,<at 1 0>]}}})) {module=ScaleRibbonWidth001}
   &ScaleColor(
    Linear Color Curve=DataInterface() {propertyValues {
      AlphaCurve {keys [<at 0 0>,<at 0.108 1>,<at 1 0>]}
      BlueCurve {keys [<at 0.0995 0.0286>,<at 0.3672 0.159>,<at 0.8398 0.5156>,<at 1 0.2083>]}
      GreenCurve {keys [<at 0.0995 0.0232>,<at 0.3672 0>,<at 0.8398 0>,<at 1 0.0083>]}
      RedCurve {keys [<at 0.0995 0.8906>,<at 0.3672 1>,<at 0.8398 0.3372>,<at 1 0>]}
     }}
    Scale Mode=::NewEnumerator2
   ) {module=ScaleColor001}
  ]
  rendererClasses [&NiagaraRibbonRendererProperties]
  renderers [{rendererClass=&NiagaraRibbonRendererProperties}]
 }
 Red_Outline {
  emitterUpdateScript [
   &EmitterState(Life Cycle Mode=::NewEnumerator1, Loop Duration=15, UseLoopDelay=false)
   &SpawnRate(SpawnRate=&Multiply_Float(A=50, B=$User.Anim_Curve_SpawnRate), Use Spawn Probability=false)
   &SpawnParticlesFromOtherEmitter(
    Attribute Reader=DataInterface() {propertyValues {EmitterBinding {bindingMode=Other, emitterName=Position001}}}
    Spawn Rate=0
    Use Maximum Age=false
    Use Minimum Age=false
   )
  ]
  particleSpawnScript [
   &InitializeParticle(
    Color=&ColorFromCurve(
     ColorCurve=DataInterface() {propertyValues {
       AlphaCurve {keys [<at 0 0 RCIM_Cubic>,<at 1 1 RCIM_Cubic>]}
       BlueCurve {keys [<at 0 1 RCIM_Cubic>,<at 1 1 RCIM_Cubic>]}
       GreenCurve {keys [<at 0 1 RCIM_Cubic>,<at 1 1 RCIM_Cubic>]}
       RedCurve {keys [<at 0 1 RCIM_Cubic>,<at 1 1 RCIM_Cubic>]}
      }}
     CurveIndex=$User.Anim_Curve_InitialOpacity
    )
    Lifetime=0.45
    Ribbon Width=25
    Ribbon Width Mode=::NewEnumerator1
    UsePositionOffset=false
    Write Material Random=true
   )
   &SampleParticlesFromOtherEmitter()
  ]
  particleUpdateScript [
   &ParticleState()
   &SolveForcesAndVelocity(Clamp Velocity=false, Limit Acceleration=false)
   &ScaleRibbonWidth(Ribbon Width Scale=&FloatFromCurve(FloatCurve=DataInterface() {propertyValues {Curve {keys [<at 0 1>,<at 1 0>]}}})) {module=ScaleRibbonWidth001}
   &ScaleColor(
    Linear Color Curve=DataInterface() {propertyValues {
      AlphaCurve {keys [<at 0 0>,<at 0.108 1>,<at 1 0>]}
      BlueCurve {keys [<at 0 0.7231>,<at 0.4285 1>,<at 1 0.2083>]}
      GreenCurve {keys [<at 0 0.6355>,<at 0.4285 0.0806>,<at 1 0.0083>]}
      RedCurve {keys [<at 0 0.0258>,<at 0.4285 0.1638>,<at 1 0>]}
     }}
    Scale Mode=::NewEnumerator2
   ) {enabled=false, module=ScaleColor001}
   &ScaleColor(
    Linear Color Curve=DataInterface() {propertyValues {
      AlphaCurve {keys [<at 0 0>,<at 0.108 1>,<at 1 0>]}
      BlueCurve {keys [<at 0.0875 0.024>,<at 0.3986 0.159>,<at 0.8189 0.5156>,<at 1 0.2083>]}
      GreenCurve {keys [<at 0.0875 0.0186>,<at 0.3986 0>,<at 0.8189 0>,<at 1 0.0083>]}
      RedCurve {keys [<at 0.0875 0.8906>,<at 0.3986 1>,<at 0.8189 0.3372>,<at 1 0>]}
     }}
    Scale Mode=::NewEnumerator2
   )
  ]
  rendererClasses [&NiagaraRibbonRendererProperties]
  renderers [{rendererClass=&NiagaraRibbonRendererProperties}]
 }
 Red_Streak {
  emitterUpdateScript [
   &EmitterState(Life Cycle Mode=::NewEnumerator1, Loop Duration=15, UseLoopDelay=false)
   &SpawnRate(SpawnRate=&Multiply_Float(A=50, B=$User.Anim_Curve_SpawnRate), Use Spawn Probability=false)
   &SpawnParticlesFromOtherEmitter(
    Attribute Reader=DataInterface() {propertyValues {EmitterBinding {bindingMode=Other, emitterName=Position001}}}
    Spawn Rate=0
    Use Maximum Age=false
    Use Minimum Age=false
   )
  ]
  particleSpawnScript [
   &InitializeParticle(
    Color=&ColorFromCurve(
     ColorCurve=DataInterface() {propertyValues {
       AlphaCurve {keys [<at 0 0 RCIM_Cubic>,<at 1 1 RCIM_Cubic>]}
       BlueCurve {keys [<at 0 1 RCIM_Cubic>,<at 1 1 RCIM_Cubic>]}
       GreenCurve {keys [<at 0 1 RCIM_Cubic>,<at 1 1 RCIM_Cubic>]}
       RedCurve {keys [<at 0 1 RCIM_Cubic>,<at 1 1 RCIM_Cubic>]}
      }}
     CurveIndex=$User.Anim_Curve_InitialOpacity
    )
    Lifetime=0.45
    Ribbon Width=25
    Ribbon Width Mode=::NewEnumerator1
    UsePositionOffset=false
    Write Material Random=true
   )
   &SampleParticlesFromOtherEmitter()
  ]
  particleUpdateScript [
   &ParticleState()
   &SolveForcesAndVelocity(Clamp Velocity=false, Limit Acceleration=false)
   &ScaleRibbonWidth(Ribbon Width Scale=&FloatFromCurve(FloatCurve=DataInterface() {propertyValues {Curve {keys [<at 0 1>,<at 1 0>]}}})) {module=ScaleRibbonWidth001}
   &ScaleColor(
    Linear Color Curve=DataInterface() {propertyValues {
      AlphaCurve {keys [<at 0 0>,<at 0.108 1>,<at 1 0>]}
      BlueCurve {keys [<at 0 0>,<at 0.4539 0>,<at 1 0.6458>]}
      GreenCurve {keys [<at 0 0.7167>,<at 0.4539 0.1537>,<at 1 0>]}
      RedCurve {keys [<at 0 1>,<at 0.4539 0.974>,<at 1 0.296>]}
     }}
    Scale Mode=::NewEnumerator2
   ) {enabled=false, module=ScaleColor001}
   &ScaleColor(
    Linear Color Curve=DataInterface() {propertyValues {
      AlphaCurve {keys [<at 0 0>,<at 0.108 1>,<at 1 0>]}
      BlueCurve {keys [<at 0.0875 0.0194>,<at 0.4029 0.159>,<at 0.8189 0.5156>,<at 1 0.2083>]}
      GreenCurve {keys [<at 0.0875 0.0139>,<at 0.4029 0>,<at 0.8189 0>,<at 1 0.0083>]}
      RedCurve {keys [<at 0.0875 0.8906>,<at 0.4029 1>,<at 0.8189 0.3372>,<at 1 0>]}
     }}
    Scale Mode=::NewEnumerator2
   )
  ]
  rendererClasses [&NiagaraRibbonRendererProperties]
  renderers [{rendererClass=&NiagaraRibbonRendererProperties}]
 }
 Yellow_Base {
  bEnabled=false
  emitterUpdateScript [
   &EmitterState(Life Cycle Mode=::NewEnumerator1, Loop Duration=15, UseLoopDelay=false)
   &SpawnRate(SpawnRate=&Multiply_Float(A=50, B=$User.Anim_Curve_SpawnRate), Use Spawn Probability=false)
   &SpawnParticlesFromOtherEmitter(
    Attribute Reader=DataInterface() {propertyValues {EmitterBinding {bindingMode=Other, emitterName=Position}}}
    Spawn Rate=0
    Use Maximum Age=false
    Use Minimum Age=false
   )
  ]
  particleSpawnScript [
   &InitializeParticle(
    Color=&ColorFromCurve(
     ColorCurve=DataInterface() {propertyValues {
       AlphaCurve {keys [<at 0 0 RCIM_Cubic>,<at 1 1 RCIM_Cubic>]}
       BlueCurve {keys [<at 0 1 RCIM_Cubic>,<at 1 1 RCIM_Cubic>]}
       GreenCurve {keys [<at 0 1 RCIM_Cubic>,<at 1 1 RCIM_Cubic>]}
       RedCurve {keys [<at 0 1 RCIM_Cubic>,<at 1 1 RCIM_Cubic>]}
      }}
     CurveIndex=$User.Anim_Curve_InitialOpacity
    )
    Lifetime=0.45
    Ribbon Width=15
    Ribbon Width Mode=::NewEnumerator1
    UsePositionOffset=false
    Write Material Random=true
   )
   &SampleParticlesFromOtherEmitter()
  ]
  particleUpdateScript [
   &ParticleState()
   &SolveForcesAndVelocity(Clamp Velocity=false, Limit Acceleration=false)
   &ScaleRibbonWidth(Ribbon Width Scale=&FloatFromCurve(FloatCurve=DataInterface() {propertyValues {Curve {keys [<at 0 1>,<at 1 0>]}}}))
   &ScaleColor(
    Linear Color Curve=DataInterface() {propertyValues {
      AlphaCurve {keys [<at 0 0>,<at 0.108 1>,<at 1 0>]}
      BlueCurve {keys [<at 0.095 0>,<at 0.4749 0>,<at 1 0.6458>]}
      GreenCurve {keys [<at 0.095 0.7167>,<at 0.4749 0.1537>,<at 1 0>]}
      RedCurve {keys [<at 0.095 1>,<at 0.4749 0.974>,<at 1 0.296>]}
     }}
    Scale Mode=::NewEnumerator2
   ) {module=ScaleColor001}
  ]
  rendererClasses [&NiagaraRibbonRendererProperties]
  renderers [{rendererClass=&NiagaraRibbonRendererProperties}]
 }
 Yellow_Outline {
  bEnabled=false
  emitterUpdateScript [
   &EmitterState(Life Cycle Mode=::NewEnumerator1, Loop Duration=15, UseLoopDelay=false)
   &SpawnRate(SpawnRate=&Multiply_Float(A=50, B=$User.Anim_Curve_SpawnRate), Use Spawn Probability=false)
   &SpawnParticlesFromOtherEmitter(
    Attribute Reader=DataInterface() {propertyValues {EmitterBinding {bindingMode=Other, emitterName=Position}}}
    Spawn Rate=0
    Use Maximum Age=false
    Use Minimum Age=false
   )
  ]
  particleSpawnScript [
   &InitializeParticle(
    Color=&ColorFromCurve(
     ColorCurve=DataInterface() {propertyValues {
       AlphaCurve {keys [<at 0 0 RCIM_Cubic>,<at 1 1 RCIM_Cubic>]}
       BlueCurve {keys [<at 0 1 RCIM_Cubic>,<at 1 1 RCIM_Cubic>]}
       GreenCurve {keys [<at 0 1 RCIM_Cubic>,<at 1 1 RCIM_Cubic>]}
       RedCurve {keys [<at 0 1 RCIM_Cubic>,<at 1 1 RCIM_Cubic>]}
      }}
     CurveIndex=$User.Anim_Curve_InitialOpacity
    )
    Lifetime=0.45
    Ribbon Width=15
    Ribbon Width Mode=::NewEnumerator1
    UsePositionOffset=false
    Write Material Random=true
   )
   &SampleParticlesFromOtherEmitter()
  ]
  particleUpdateScript [
   &ParticleState()
   &SolveForcesAndVelocity(Clamp Velocity=false, Limit Acceleration=false)
   &ScaleRibbonWidth(Ribbon Width Scale=&FloatFromCurve(FloatCurve=DataInterface() {propertyValues {Curve {keys [<at 0 1>,<at 1 0>]}}})) {module=ScaleRibbonWidth001}
   &ScaleColor(
    Linear Color Curve=DataInterface() {propertyValues {
      AlphaCurve {keys [<at 0 0>,<at 0.108 1>,<at 1 0>]}
      BlueCurve {keys [<at 0 0.7231>,<at 0.4285 1>,<at 1 0.2083>]}
      GreenCurve {keys [<at 0 0.6355>,<at 0.4285 0.0806>,<at 1 0.0083>]}
      RedCurve {keys [<at 0 0.0258>,<at 0.4285 0.1638>,<at 1 0>]}
     }}
    Scale Mode=::NewEnumerator2
   ) {enabled=false, module=ScaleColor001}
   &ScaleColor(
    Linear Color Curve=DataInterface() {propertyValues {
      AlphaCurve {keys [<at 0 0>,<at 0.108 1>,<at 1 0>]}
      BlueCurve {keys [<at 0.0875 0.0424>,<at 0.3986 0.159>,<at 0.8189 0.5156>,<at 1 0.2083>]}
      GreenCurve {keys [<at 0.0875 0.0371>,<at 0.3986 0>,<at 0.8189 0>,<at 1 0.0083>]}
      RedCurve {keys [<at 0.0875 0.8906>,<at 0.3986 1>,<at 0.8189 0.3372>,<at 1 0>]}
     }}
    Scale Mode=::NewEnumerator2
   )
  ]
  rendererClasses [&NiagaraRibbonRendererProperties]
  renderers [{rendererClass=&NiagaraRibbonRendererProperties}]
 }
 Yellow_Streak {
  bEnabled=false
  emitterUpdateScript [
   &EmitterState(Life Cycle Mode=::NewEnumerator1, Loop Duration=15, UseLoopDelay=false)
   &SpawnRate(SpawnRate=&Multiply_Float(A=50, B=$User.Anim_Curve_SpawnRate), Use Spawn Probability=false)
   &SpawnParticlesFromOtherEmitter(
    Attribute Reader=DataInterface() {propertyValues {EmitterBinding {bindingMode=Other, emitterName=Position}}}
    Spawn Rate=0
    Use Maximum Age=false
    Use Minimum Age=false
   )
  ]
  particleSpawnScript [
   &InitializeParticle(
    Color=&ColorFromCurve(
     ColorCurve=DataInterface() {propertyValues {
       AlphaCurve {keys [<at 0 0 RCIM_Cubic>,<at 1 1 RCIM_Cubic>]}
       BlueCurve {keys [<at 0 1 RCIM_Cubic>,<at 1 1 RCIM_Cubic>]}
       GreenCurve {keys [<at 0 1 RCIM_Cubic>,<at 1 1 RCIM_Cubic>]}
       RedCurve {keys [<at 0 1 RCIM_Cubic>,<at 1 1 RCIM_Cubic>]}
      }}
     CurveIndex=$User.Anim_Curve_InitialOpacity
    )
    Lifetime=0.45
    Ribbon Width=15
    Ribbon Width Mode=::NewEnumerator1
    UsePositionOffset=false
    Write Material Random=true
   )
   &SampleParticlesFromOtherEmitter()
  ]
  particleUpdateScript [
   &ParticleState()
   &SolveForcesAndVelocity(Clamp Velocity=false, Limit Acceleration=false)
   &ScaleRibbonWidth(Ribbon Width Scale=&FloatFromCurve(FloatCurve=DataInterface() {propertyValues {Curve {keys [<at 0 1>,<at 1 0>]}}})) {module=ScaleRibbonWidth001}
   &ScaleColor(
    Linear Color Curve=DataInterface() {propertyValues {
      AlphaCurve {keys [<at 0 0>,<at 0.108 1>,<at 1 0>]}
      BlueCurve {keys [<at 0 0>,<at 0.4539 0>,<at 1 0.6458>]}
      GreenCurve {keys [<at 0 0.7167>,<at 0.4539 0.1537>,<at 1 0>]}
      RedCurve {keys [<at 0 1>,<at 0.4539 0.974>,<at 1 0.296>]}
     }}
    Scale Mode=::NewEnumerator2
   ) {module=ScaleColor001}
  ]
  rendererClasses [&NiagaraRibbonRendererProperties]
  renderers [{rendererClass=&NiagaraRibbonRendererProperties}]
 }
}
summary {
 systemName=NS_Trail_A
 emitters [
  {bEnabled=false, emitterName=Yellow_Base, rendererClasses [&NiagaraRibbonRendererProperties]}
  {bEnabled=false, emitterName=Position, rendererClasses [&NiagaraSpriteRendererProperties]}
  {bEnabled=false, emitterName=Yellow_Streak, rendererClasses [&NiagaraRibbonRendererProperties]}
  {bEnabled=false, emitterName=Yellow_Outline, rendererClasses [&NiagaraRibbonRendererProperties]}
  {bEnabled=false, emitterName=L_Small_Yellow_Base, rendererClasses [&NiagaraRibbonRendererProperties]}
  {bEnabled=false, emitterName=L_Position, rendererClasses [&NiagaraSpriteRendererProperties]}
  {bEnabled=false, emitterName=R_Position, rendererClasses [&NiagaraSpriteRendererProperties]}
  {bEnabled=false, emitterName=R_Small_Yellow_Base, rendererClasses [&NiagaraRibbonRendererProperties]}
  {bEnabled=false, emitterName=L_Small_Yellow_Base001, rendererClasses [&NiagaraRibbonRendererProperties]}
  {bEnabled=false, emitterName=L_Position_Spawn, rendererClasses [&NiagaraSpriteRendererProperties]}
  {bEnabled=false, emitterName=R_Position_Spawn, rendererClasses [&NiagaraSpriteRendererProperties]}
  {bEnabled=false, emitterName=R_Small_Yellow_Base001, rendererClasses [&NiagaraRibbonRendererProperties]}
  {emitterName=Red_Base, rendererClasses [&NiagaraRibbonRendererProperties]}
  {emitterName=Position001, rendererClasses [&NiagaraSpriteRendererProperties]}
  {emitterName=Red_Outline, rendererClasses [&NiagaraRibbonRendererProperties]}
  {emitterName=Red_Streak, rendererClasses [&NiagaraRibbonRendererProperties]}
 ]
 userVariables [
  User.Anim_Curve_InitialOpacity {type=NiagaraFloat, defaultValue=1, description=""}
  User.Anim_Curve_SmallTrail_InitialOpacity {type=NiagaraFloat, defaultValue=0, description=""}
  User.Anim_Curve_SpawnRate {type=NiagaraFloat, defaultValue=0, description=""}
  User.bIsFrontEnd {type=NiagaraBool}
  User.bIsFrontEndPreview {type=NiagaraBool}
  User.FX_WeaponCaseSize {type=NiagaraInt32, defaultValue=1}
  User.NormalizedNotifyProgress {type=NiagaraFloat, defaultValue=0, description=""}
  User.SystemScale {type=NiagaraFloat, defaultValue=1, description=""}
 ]
}
