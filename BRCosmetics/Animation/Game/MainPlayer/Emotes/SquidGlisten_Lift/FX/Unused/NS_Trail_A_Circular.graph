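# Niagara trail system for the SquidGlisten_Lift emote, filed under Unused, drawing paired ribbons from the weapon_l bone of the attach parent's skeletal mesh.
# Two long-lived Position emitters follow that bone at opposite X offsets and act as leaders that ribbon emitters sample via particle reads.
# Active red-to-purple ribbons taper and fade, with user animation-curve parameters driving spawn rate and initial opacity.
# The yellow base, outline and streak emitters are disabled.
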NiagaraSystem /BRCosmetics/Animation/Game/MainPlayer/Emotes/SquidGlisten_Lift/FX/Unused/NS_Trail_A_Circular
dependencies {
 usedDataInterfaces [
  &NiagaraDataInterfaceParticleRead
  &NiagaraDataInterfaceColorCurve
  &NiagaraDataInterfaceCurve
  &NiagaraDataInterfaceSkeletalMesh
  &NiagaraDataInterfaceArrayFloat
 ]
 usedDynamicInputs [&FNTimeOfDayColor, &SelectFloatFromArray]
 usedModules [
  &SystemState
  &self:SystemSpawnScript.SystemScriptSource.SystemScriptGraph.NiagaraNodeAssignment_0.SetVariables_#1
 ]
}
emitters {
 Position {
  emitterUpdateScript [
   &EmitterState(Life Cycle Mode=::NewEnumerator1, Loop Duration=15, UseLoopDelay=false)
   &SpawnBurst_Instantaneous(Use Spawn Probability=false, UseLoopCountLimit=false)
  ]
  particleSpawnScript [&InitializeParticle(Color Mode=::NewEnumerator0, Lifetime=15, UsePositionOffset=true, Write Material Random=true)]
  particleUpdateScript [
   &ParticleState()
   &SkeletalMeshLocation(
    Bone Sampling Mode=::NewEnumerator4
    MaskSpawnGroup=false
    Orientation Sampling=::NewEnumerator1
    Sampled Position Offset=<16 0 0>
    Skeletal Mesh=DataInterface() {propertyValues {
      DefaultMesh=null
      ExcludeBoneName=root
      OptionalComponentClass=null
      PreviewMesh=null
      SkinningMode=SkinOnTheFly
      SoftSourceActor=null
      SourceMode=AttachParent
      UvSetIndex=0
      WholeMeshLOD=-1
      bExcludeBone=false
      bReadDeformedGeometry=true
      bRequireCurrentFrameData=true
      ComponentTags []
      FilteredBones [weapon_l]
      FilteredSockets []
      MeshUserParameter {parameter {name=None, typeDefHandle {registeredTypeIndex=1}, varData []}}
      SamplingRegions []
     }}
   ) {module=SkeletalMeshLocation001}
  ]
  rendererClasses [&NiagaraSpriteRendererProperties]
  renderers [{rendererClass=&NiagaraSpriteRendererProperties}]
 }
 Position001 {
  emitterUpdateScript [
   &EmitterState(Life Cycle Mode=::NewEnumerator1, Loop Duration=15, UseLoopDelay=false)
   &SpawnBurst_Instantaneous(Use Spawn Probability=false, UseLoopCountLimit=false)
  ]
  particleSpawnScript [&InitializeParticle(Color Mode=::NewEnumerator0, Lifetime=15, UsePositionOffset=true, Write Material Random=true)]
  particleUpdateScript [
   &ParticleState()
   &SkeletalMeshLocation(
    Bone Sampling Mode=::NewEnumerator4
    MaskSpawnGroup=false
    Orientation Sampling=::NewEnumerator1
    Sampled Position Offset=<-16 0 0>
    Skeletal Mesh=DataInterface() {propertyValues {
      DefaultMesh=null
      ExcludeBoneName=root
      OptionalComponentClass=null
      PreviewMesh=null
      SkinningMode=SkinOnTheFly
      SoftSourceActor=null
      SourceMode=AttachParent
      UvSetIndex=0
      WholeMeshLOD=-1
      bExcludeBone=false
      bReadDeformedGeometry=true
      bRequireCurrentFrameData=true
      ComponentTags []
      FilteredBones [weapon_l]
      FilteredSockets []
      MeshUserParameter {parameter {name=None, typeDefHandle {registeredTypeIndex=1}, varData []}}
      SamplingRegions []
     }}
   ) {module=SkeletalMeshLocation001}
  ]
  rendererClasses [&NiagaraSpriteRendererProperties]
  renderers [{rendererClass=&NiagaraSpriteRendererProperties}]
 }
 Red_Base {
  emitterUpdateScript [
   &EmitterState(Life Cycle Mode=::NewEnumerator1, Loop Duration=15, UseLoopDelay=false)
   &SpawnRate(SpawnRate=&Multiply_Float(A=50, B=$User.Anim_Curve_SpawnRate), Use Spawn Probability=false)
   &SpawnParticlesFromOtherEmitter(
    Attribute Reader=DataInterface() {propertyValues {EmitterBinding {bindingMode=Other, emitterName=Position}}}
    Spawn Rate=0
    Use Maximum Age=false
    Use Minimum Age=false
   )
  ]
  particleSpawnScript [
   &InitializeParticle(
    Color=&ColorFromCurve(
     ColorCurve=DataInterface() {propertyValues {
       AlphaCurve {keys [<at 0 0 RCIM_Cubic>,<at 1 1 RCIM_Cubic>]}
       BlueCurve {keys [<at 0 1 RCIM_Cubic>,<at 1 1 RCIM_Cubic>]}
       GreenCurve {keys [<at 0 1 RCIM_Cubic>,<at 1 1 RCIM_Cubic>]}
       RedCurve {keys [<at 0 1 RCIM_Cubic>,<at 1 1 RCIM_Cubic>]}
      }}
     CurveIndex=$User.Anim_Curve_InitialOpacity
    )
    Lifetime=0.25
    Ribbon Width=2
    Ribbon Width Mode=::NewEnumerator1
    UsePositionOffset=false
    Write Material Random=true
   )
   &SampleParticlesFromOtherEmitter()
  ]
  particleUpdateScript [
   &ParticleState()
   &SolveForcesAndVelocity(Clamp Velocity=false, Limit Acceleration=false)
   &ScaleRibbonWidth(Ribbon Width Scale=&FloatFromCurve(FloatCurve=DataInterface() {propertyValues {Curve {keys [<at 0 1>,<at 1 0>]}}})) {module=ScaleRibbonWidth001}
   &ScaleColor(
    Linear Color Curve=DataInterface() {propertyValues {
      AlphaCurve {keys [<at 0 0>,<at 0.108 1>,<at 1 0>]}
      BlueCurve {keys [<at 0.0995 0.0286>,<at 0.3672 0.159>,<at 0.8398 0.5156>,<at 1 0.2083>]}
      GreenCurve {keys [<at 0.0995 0.0232>,<at 0.3672 0>,<at 0.8398 0>,<at 1 0.0083>]}
      RedCurve {keys [<at 0.0995 0.8906>,<at 0.3672 1>,<at 0.8398 0.3372>,<at 1 0>]}
     }}
    Scale Mode=::NewEnumerator2
   ) {module=ScaleColor001}
  ]
  rendererClasses [&NiagaraRibbonRendererProperties]
  renderers [{rendererClass=&NiagaraRibbonRendererProperties}]
 }
 Red_Base001 {
  emitterUpdateScript [
   &EmitterState(Life Cycle Mode=::NewEnumerator1, Loop Duration=15, UseLoopDelay=false)
   &SpawnRate(SpawnRate=&Multiply_Float(A=50, B=$User.Anim_Curve_SpawnRate), Use Spawn Probability=false)
   &SpawnParticlesFromOtherEmitter(
    Attribute Reader=DataInterface() {propertyValues {EmitterBinding {bindingMode=Other, emitterName=Position001}}}
    Spawn Rate=0
    Use Maximum Age=false
    Use Minimum Age=false
   )
  ]
  particleSpawnScript [
   &InitializeParticle(
    Color=&ColorFromCurve(
     ColorCurve=DataInterface() {propertyValues {
       AlphaCurve {keys [<at 0 0 RCIM_Cubic>,<at 1 1 RCIM_Cubic>]}
       BlueCurve {keys [<at 0 1 RCIM_Cubic>,<at 1 1 RCIM_Cubic>]}
       GreenCurve {keys [<at 0 1 RCIM_Cubic>,<at 1 1 RCIM_Cubic>]}
       RedCurve {keys [<at 0 1 RCIM_Cubic>,<at 1 1 RCIM_Cubic>]}
      }}
     CurveIndex=$User.Anim_Curve_InitialOpacity
    )
    Lifetime=0.25
    Ribbon Width=2
    Ribbon Width Mode=::NewEnumerator1
    UsePositionOffset=false
    Write Material Random=true
   )
   &SampleParticlesFromOtherEmitter()
  ]
  particleUpdateScript [
   &ParticleState()
   &SolveForcesAndVelocity(Clamp Velocity=false, Limit Acceleration=false)
   &ScaleRibbonWidth(Ribbon Width Scale=&FloatFromCurve(FloatCurve=DataInterface() {propertyValues {Curve {keys [<at 0 1>,<at 1 0>]}}})) {module=ScaleRibbonWidth001}
   &ScaleColor(
    Linear Color Curve=DataInterface() {propertyValues {
      AlphaCurve {keys [<at 0 0>,<at 0.108 1>,<at 1 0>]}
      BlueCurve {keys [<at 0.0995 0.0286>,<at 0.3672 0.159>,<at 0.8398 0.5156>,<at 1 0.2083>]}
      GreenCurve {keys [<at 0.0995 0.0232>,<at 0.3672 0>,<at 0.8398 0>,<at 1 0.0083>]}
      RedCurve {keys [<at 0.0995 0.8906>,<at 0.3672 1>,<at 0.8398 0.3372>,<at 1 0>]}
     }}
    Scale Mode=::NewEnumerator2
   ) {module=ScaleColor001}
  ]
  rendererClasses [&NiagaraRibbonRendererProperties]
  renderers [{rendererClass=&NiagaraRibbonRendererProperties}]
 }
 Yellow_Base {
  bEnabled=false
  emitterUpdateScript [
   &EmitterState(Life Cycle Mode=::NewEnumerator1, Loop Duration=15, UseLoopDelay=false)
   &SpawnRate(SpawnRate=&Multiply_Float(A=50, B=$User.Anim_Curve_SpawnRate), Use Spawn Probability=false)
   &SpawnParticlesFromOtherEmitter(
    Attribute Reader=DataInterface() {propertyValues {EmitterBinding {bindingMode=Other, emitterName=Position}}}
    Spawn Rate=0
    Use Maximum Age=false
    Use Minimum Age=false
   )
  ]
  particleSpawnScript [
   &InitializeParticle(
    Color=&ColorFromCurve(
     ColorCurve=DataInterface() {propertyValues {
       AlphaCurve {keys [<at 0 0 RCIM_Cubic>,<at 1 1 RCIM_Cubic>]}
       BlueCurve {keys [<at 0 1 RCIM_Cubic>,<at 1 1 RCIM_Cubic>]}
       GreenCurve {keys [<at 0 1 RCIM_Cubic>,<at 1 1 RCIM_Cubic>]}
       RedCurve {keys [<at 0 1 RCIM_Cubic>,<at 1 1 RCIM_Cubic>]}
      }}
     CurveIndex=$User.Anim_Curve_InitialOpacity
    )
    Lifetime=0.25
    Ribbon Width=2
    Ribbon Width Mode=::NewEnumerator1
    UsePositionOffset=false
    Write Material Random=true
   )
   &SampleParticlesFromOtherEmitter()
  ]
  particleUpdateScript [
   &ParticleState()
   &SolveForcesAndVelocity(Clamp Velocity=false, Limit Acceleration=false)
   &ScaleRibbonWidth(Ribbon Width Scale=&FloatFromCurve(FloatCurve=DataInterface() {propertyValues {Curve {keys [<at 0 1>,<at 1 0>]}}}))
   &ScaleColor(
    Linear Color Curve=DataInterface() {propertyValues {
      AlphaCurve {keys [<at 0 0>,<at 0.108 1>,<at 1 0>]}
      BlueCurve {keys [<at 0.095 0>,<at 0.4749 0>,<at 1 0.6458>]}
      GreenCurve {keys [<at 0.095 0.7167>,<at 0.4749 0.1537>,<at 1 0>]}
      RedCurve {keys [<at 0.095 1>,<at 0.4749 0.974>,<at 1 0.296>]}
     }}
    Scale Mode=::NewEnumerator2
   ) {module=ScaleColor001}
  ]
  rendererClasses [&NiagaraRibbonRendererProperties]
  renderers [{rendererClass=&NiagaraRibbonRendererProperties}]
 }
 Yellow_Base001 {
  bEnabled=false
  emitterUpdateScript [
   &EmitterState(Life Cycle Mode=::NewEnumerator1, Loop Duration=15, UseLoopDelay=false)
   &SpawnRate(SpawnRate=&Multiply_Float(A=50, B=$User.Anim_Curve_SpawnRate), Use Spawn Probability=false)
   &SpawnParticlesFromOtherEmitter(
    Attribute Reader=DataInterface() {propertyValues {EmitterBinding {bindingMode=Other, emitterName=Position001}}}
    Spawn Rate=0
    Use Maximum Age=false
    Use Minimum Age=false
   )
  ]
  particleSpawnScript [
   &InitializeParticle(
    Color=&ColorFromCurve(
     ColorCurve=DataInterface() {propertyValues {
       AlphaCurve {keys [<at 0 0 RCIM_Cubic>,<at 1 1 RCIM_Cubic>]}
       BlueCurve {keys [<at 0 1 RCIM_Cubic>,<at 1 1 RCIM_Cubic>]}
       GreenCurve {keys [<at 0 1 RCIM_Cubic>,<at 1 1 RCIM_Cubic>]}
       RedCurve {keys [<at 0 1 RCIM_Cubic>,<at 1 1 RCIM_Cubic>]}
      }}
     CurveIndex=$User.Anim_Curve_InitialOpacity
    )
    Lifetime=0.25
    Ribbon Width=2
    Ribbon Width Mode=::NewEnumerator1
    UsePositionOffset=false
    Write Material Random=true
   )
   &SampleParticlesFromOtherEmitter()
  ]
  particleUpdateScript [
   &ParticleState()
   &SolveForcesAndVelocity(Clamp Velocity=false, Limit Acceleration=false)
   &ScaleRibbonWidth(Ribbon Width Scale=&FloatFromCurve(FloatCurve=DataInterface() {propertyValues {Curve {keys [<at 0 1>,<at 1 0>]}}}))
   &ScaleColor(
    Linear Color Curve=DataInterface() {propertyValues {
      AlphaCurve {keys [<at 0 0>,<at 0.108 1>,<at 1 0>]}
      BlueCurve {keys [<at 0.095 0>,<at 0.4749 0>,<at 1 0.6458>]}
      GreenCurve {keys [<at 0.095 0.7167>,<at 0.4749 0.1537>,<at 1 0>]}
      RedCurve {keys [<at 0.095 1>,<at 0.4749 0.974>,<at 1 0.296>]}
     }}
    Scale Mode=::NewEnumerator2
   ) {module=ScaleColor001}
  ]
  rendererClasses [&NiagaraRibbonRendererProperties]
  renderers [{rendererClass=&NiagaraRibbonRendererProperties}]
 }
 Yellow_Outline {
  bEnabled=false
  emitterUpdateScript [
   &EmitterState(Life Cycle Mode=::NewEnumerator1, Loop Duration=15, UseLoopDelay=false)
   &SpawnRate(SpawnRate=&Multiply_Float(A=50, B=$User.Anim_Curve_SpawnRate), Use Spawn Probability=false)
   &SpawnParticlesFromOtherEmitter(
    Attribute Reader=DataInterface() {propertyValues {EmitterBinding {bindingMode=Other, emitterName=Position}}}
    Spawn Rate=0
    Use Maximum Age=false
    Use Minimum Age=false
   )
  ]
  particleSpawnScript [
   &InitializeParticle(
    Color=&ColorFromCurve(
     ColorCurve=DataInterface() {propertyValues {
       AlphaCurve {keys [<at 0 0 RCIM_Cubic>,<at 1 1 RCIM_Cubic>]}
       BlueCurve {keys [<at 0 1 RCIM_Cubic>,<at 1 1 RCIM_Cubic>]}
       GreenCurve {keys [<at 0 1 RCIM_Cubic>,<at 1 1 RCIM_Cubic>]}
       RedCurve {keys [<at 0 1 RCIM_Cubic>,<at 1 1 RCIM_Cubic>]}
      }}
     CurveIndex=$User.Anim_Curve_InitialOpacity
    )
    Lifetime=0.45
    Ribbon Width=10
    Ribbon Width Mode=::NewEnumerator1
    UsePositionOffset=false
    Write Material Random=true
   )
   &SampleParticlesFromOtherEmitter()
  ]
  particleUpdateScript [
   &ParticleState()
   &SolveForcesAndVelocity(Clamp Velocity=false, Limit Acceleration=false)
   &ScaleRibbonWidth(Ribbon Width Scale=&FloatFromCurve(FloatCurve=DataInterface() {propertyValues {Curve {keys [<at 0 1>,<at 1 0>]}}})) {module=ScaleRibbonWidth001}
   &ScaleColor(
    Linear Color Curve=DataInterface() {propertyValues {
      AlphaCurve {keys [<at 0 0>,<at 0.108 1>,<at 1 0>]}
      BlueCurve {keys [<at 0 0.7231>,<at 0.4285 1>,<at 1 0.2083>]}
      GreenCurve {keys [<at 0 0.6355>,<at 0.4285 0.0806>,<at 1 0.0083>]}
      RedCurve {keys [<at 0 0.0258>,<at 0.4285 0.1638>,<at 1 0>]}
     }}
    Scale Mode=::NewEnumerator2
   ) {enabled=false, module=ScaleColor001}
   &ScaleColor(
    Linear Color Curve=DataInterface() {propertyValues {
      AlphaCurve {keys [<at 0 0>,<at 0.108 1>,<at 1 0>]}
      BlueCurve {keys [<at 0.0875 0.0424>,<at 0.3986 0.159>,<at 0.8189 0.5156>,<at 1 0.2083>]}
      GreenCurve {keys [<at 0.0875 0.0371>,<at 0.3986 0>,<at 0.8189 0>,<at 1 0.0083>]}
      RedCurve {keys [<at 0.0875 0.8906>,<at 0.3986 1>,<at 0.8189 0.3372>,<at 1 0>]}
     }}
    Scale Mode=::NewEnumerator2
   )
  ]
  rendererClasses [&NiagaraRibbonRendererProperties]
  renderers [{rendererClass=&NiagaraRibbonRendererProperties}]
 }
 Yellow_Streak {
  bEnabled=false
  emitterUpdateScript [
   &EmitterState(Life Cycle Mode=::NewEnumerator1, Loop Duration=15, UseLoopDelay=false)
   &SpawnRate(SpawnRate=&Multiply_Float(A=50, B=$User.Anim_Curve_SpawnRate), Use Spawn Probability=false)
   &SpawnParticlesFromOtherEmitter(
    Attribute Reader=DataInterface() {propertyValues {EmitterBinding {bindingMode=Other, emitterName=Position}}}
    Spawn Rate=0
    Use Maximum Age=false
    Use Minimum Age=false
   )
  ]
  particleSpawnScript [
   &InitializeParticle(
    Color=&ColorFromCurve(
     ColorCurve=DataInterface() {propertyValues {
       AlphaCurve {keys [<at 0 0 RCIM_Cubic>,<at 1 1 RCIM_Cubic>]}
       BlueCurve {keys [<at 0 1 RCIM_Cubic>,<at 1 1 RCIM_Cubic>]}
       GreenCurve {keys [<at 0 1 RCIM_Cubic>,<at 1 1 RCIM_Cubic>]}
       RedCurve {keys [<at 0 1 RCIM_Cubic>,<at 1 1 RCIM_Cubic>]}
      }}
     CurveIndex=$User.Anim_Curve_InitialOpacity
    )
    Lifetime=0.45
    Ribbon Width=6
    Ribbon Width Mode=::NewEnumerator1
    UsePositionOffset=false
    Write Material Random=true
   )
   &SampleParticlesFromOtherEmitter()
  ]
  particleUpdateScript [
   &ParticleState()
   &SolveForcesAndVelocity(Clamp Velocity=false, Limit Acceleration=false)
   &ScaleRibbonWidth(Ribbon Width Scale=&FloatFromCurve(FloatCurve=DataInterface() {propertyValues {Curve {keys [<at 0 1>,<at 1 0>]}}})) {module=ScaleRibbonWidth001}
   &ScaleColor(
    Linear Color Curve=DataInterface() {propertyValues {
      AlphaCurve {keys [<at 0 0>,<at 0.108 1>,<at 1 0>]}
      BlueCurve {keys [<at 0 0>,<at 0.4539 0>,<at 1 0.6458>]}
      GreenCurve {keys [<at 0 0.7167>,<at 0.4539 0.1537>,<at 1 0>]}
      RedCurve {keys [<at 0 1>,<at 0.4539 0.974>,<at 1 0.296>]}
     }}
    Scale Mode=::NewEnumerator2
   ) {module=ScaleColor001}
  ]
  rendererClasses [&NiagaraRibbonRendererProperties]
  renderers [{rendererClass=&NiagaraRibbonRendererProperties}]
 }
}
summary {
 systemName=NS_Trail_A_Circular
 emitters [
  {bEnabled=false, emitterName=Yellow_Base, rendererClasses [&NiagaraRibbonRendererProperties]}
  {emitterName=Position, rendererClasses [&NiagaraSpriteRendererProperties]}
  {bEnabled=false, emitterName=Yellow_Streak, rendererClasses [&NiagaraRibbonRendererProperties]}
  {bEnabled=false, emitterName=Yellow_Outline, rendererClasses [&NiagaraRibbonRendererProperties]}
  {emitterName=Position001, rendererClasses [&NiagaraSpriteRendererProperties]}
  {bEnabled=false, emitterName=Yellow_Base001, rendererClasses [&NiagaraRibbonRendererProperties]}
  {emitterName=Red_Base, rendererClasses [&NiagaraRibbonRendererProperties]}
  {emitterName=Red_Base001, rendererClasses [&NiagaraRibbonRendererProperties]}
 ]
 userVariables [
  User.Anim_Curve_Initial_Opacity {type=NiagaraFloat, defaultValue=0, description=""}
  User.Anim_Curve_InitialOpacity {type=NiagaraFloat, defaultValue=1, description=""}
  User.Anim_Curve_SpawnRate {type=NiagaraFloat, defaultValue=0, description=""}
  User.bIsFrontEnd {type=NiagaraBool}
  User.bIsFrontEndPreview {type=NiagaraBool}
  User.FX_WeaponCaseSize {type=NiagaraInt32, defaultValue=1}
  User.NormalizedNotifyProgress {type=NiagaraFloat, defaultValue=0, description=""}
  User.SystemScale {type=NiagaraFloat, defaultValue=1, description=""}
 ]
}
